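# Niagara bling effect for the InkHoop player emote, drawing a short, faint, pale sweep on a mesh.
# The sweep emitter burst-spawns a small, offset, oriented mesh particle with a slight camera offset and a half-second lifetime.
# An eased curve drives a dynamic material parameter over that lifetime, while a time-of-day scalar dims alpha, most at night.
# A Sparkle emitter is also listed but has no renderers or modules shown.
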NiagaraSystem /BRCosmetics/Animation/Game/MainPlayer/Emotes/InkHoop/FX/NS_Emote_InkHoop_Bling
dependencies {
 usedDataInterfaces [&NiagaraDataInterfaceCurve]
 usedDynamicInputs [&MakeLinearColor]
 usedModules [
  &SystemState
  &self:SystemSpawnScript.SystemScriptSource.SystemScriptGraph.NiagaraNodeAssignment_0.SetVariables_#1
 ]
}
emitters {
 Sparkle {rendererClasses [], renderers []}
 sweep {
  emitterUpdateScript [
   &EmitterState(Life Cycle Mode=::NewEnumerator1, Loop Behavior=::NewEnumerator1, Loop Duration=0.8, UseLoopDelay=false)
   &SpawnBurst_Instantaneous(Use Spawn Probability=false, UseLoopCountLimit=false)
  ]
  particleSpawnScript [
   &InitializeParticle(
    Color=<rgba 0.8941 0.9259 1 0.15>
    Lifetime=0.5
    Mesh Scale=&Multiply_VectorByFloat(Vector=<1 1 1.2>, Float=0.0475)
    Mesh Scale Mode=::NewEnumerator4
    Position Offset=<-9 -2 0.1>
    UsePositionOffset=true
    Write Material Random=true
   )
   &CameraOffset(Camera Offset Amount=1)
   &InitialMeshOrientation(Mesh Orientation Mode=::NewEnumerator2, Orientation Vector=<0 0 1>, Rotation=<0 0.5 0>, Use Rotation Vector=true)
  ]
  particleUpdateScript [
   &ParticleState()
   &DynamicMaterialParameters(
    Index 0 Param 1=&FloatFromCurve(FloatCurve=DataInterface() {propertyValues {Curve {keys [{interpMode=RCIM_Cubic, tangentMode=RCTM_Break, time=0, value=0}, {arriveTangent=2.3933, tangentMode=RCTM_User, time=1, value=1}]}}})
    Index 0 Param 2=1
    Index 0 Param 3=-0.0142
    Module.Index0Float0Write=true
    Module.Index0Float1Write=true
    Module.Index0Float2Write=true
    Module.Index0Float3Write=true
   )
   &SolveForcesAndVelocity(Clamp Velocity=false, Limit Acceleration=false)
   &ScaleColor(
    Scale Alpha=&FNTimeOfDayScalar(Night Scale=0.7, Morning Scale=0.9, Evening Scale=0.9)
    ScaleA=true
    ScaleRGB=false
   )
  ]
  rendererClasses [&NiagaraMeshRendererProperties]
  renderers [{rendererClass=&NiagaraMeshRendererProperties, rendererIndex=0}]
 }
}
summary {
 systemName=NS_Emote_InkHoop_Bling
 emitters [{emitterName=sweep, rendererClasses [&NiagaraMeshRendererProperties]}, {emitterName=Sparkle, rendererClasses []}]
 userVariables []
}
