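# Niagara stomp-impact effect for the main player's InkHoop emote.
# Impact_Spikes bursts short-lived SubUV sprites from a tiny ring after a slight delay, pushing them outward under drag while they stretch and fade.
# Sprites face world up, draw through two sprite renderers, and take tint and opacity from time-of-day functions.
# A second emitter, SmokeLWE, is listed without modules or renderers.
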NiagaraSystem /BRCosmetics/Animation/Game/MainPlayer/Emotes/InkHoop/FX/NS_Emote_InkHoop_Stomps
dependencies {
 usedDataInterfaces [&NiagaraDataInterfaceCurve, &NiagaraDataInterfaceVector2DCurve]
 usedDynamicInputs []
 usedModules [&SystemState]
}
emitters {
 Impact_Spikes {
  emitterUpdateScript [
   &EmitterState(Life Cycle Mode=::NewEnumerator1, Loop Behavior=::NewEnumerator1, Loop Delay=0.03, Loop Duration=1, UseLoopDelay=true)
   &SpawnBurst_Instantaneous(Spawn Count=13, Use Spawn Probability=false, UseLoopCountLimit=false)
  ]
  particleSpawnScript [
   &InitializeParticle(
    Color=&HueShiftLinearColor(Hue=0, Saturation=1, Value=&RandomRangeFloat(Minimum=0.4, Maximum=0.8), Linear Color=$Particles.Color)
    Lifetime Max=0.3
    Lifetime Min=0.2
    Lifetime Mode=::NewEnumerator1
    Sprite Size Max=<30 40>
    Sprite Size Min=<10 12.5>
    Sprite Size Mode=::NewEnumerator2
    UsePositionOffset=false
    Write Material Random=true
   )
   &TorusLocation(
    Align Torus With Axis=false
    Coordinate Space:ENiagaraCoordinateSpace=::World
    Handle Radius=0
    Large Radius=0.16
    MaskSpawnGroup=false
    Radius Position=0
    Torus Distribution Mode:ENiagaraTorusDistributionMode=::NewEnumerator1
    U Position:NiagaraFloat=&RandomRangeFloat(Minimum=0, Maximum=1, Randomness Mode=::NewEnumerator2, Random Seed=?)
   )
   &AddVelocityFromPoint(
    Module.Use Falloff Distance=false
    Origin Offset=&RandomRangeVector(Minimum=<0 0 0>, Maximum=<0 0 -0.2>)
    Origin Offset Coordinate Space=::World
    Velocity Strength=&RandomRangeFloat(Minimum=32, Maximum=64)
   )
   &SubUVAnimation(End Frame=3, SubUV Animation Mode:ENiagara_SubUVLookupMode=::NewEnumerator2)
  ]
  particleUpdateScript [
   &ParticleState()
   &Drag(Drag=3, Use Linear Drag=true, Use Rotational Drag=false)
   &SolveForcesAndVelocity(Clamp Velocity=false, Limit Acceleration=false)
   &ScaleSpriteSize(
    Non-Uniform Curve Scale=&Multiply_Vector2D_ByFloat(Float=0.8, Vector2D=<1 1>)
    Non-Uniform Curve Sprite Scale=DataInterface() {propertyValues {
      XCurve {keys [{tangentMode=RCTM_User, time=0, value=1}, {tangentMode=RCTM_User, time=1, value=0.8}]}
      YCurve {keys [{tangentMode=RCTM_User, time=0, value=0.5}, {tangentMode=RCTM_User, time=1, value=1.7}]}
     }}
    Scale Sprite Size Mode=::NewEnumerator3
   )
   &ScaleColor(
    Scale Alpha=&Multiply_Float(
     A=&FloatFromCurve(
      FloatCurve=DataInterface() {propertyValues {Curve {keys [<at 0 0>,<at 0.1 1>,<at 0.2 0.8>,<at 1 0>]}}}
      Scale Curve=&FNTimeOfDayScalar(Night Scale=0.4, Morning Scale=0.3, Evening Scale=0.5)
     )
     B=0.1
    )
    Scale RGB=&MakeVectorFromLinearColorRGB(Color=&FN_Select_Color_from_Time_Of_Day(Morning=<rgba 1 0.7336 0.968 1>, Day=<rgba 1 0.8 0.8 1>, Evening=<rgba 0.7114 0.9396 1 1>, Night=<rgba 0.6619 0.7344 1 1>))
    ScaleA=true
    ScaleRGB=true
   )
   /SetVariables(Particles.Color2=&MultiplyLinearColor(Linear Color=$Particles.Color, Scale Factor=<1 1 1 0.2>)) {isSetParameters=true}
   &SpriteFacingAndAlignment(Facing Coordinate Space=::World, Sprite Facing=<0 0 1>, WriteAlignment=false, WriteFacing=true)
  ]
  rendererClasses [&NiagaraSpriteRendererProperties]
  renderers [
   {rendererClass=&NiagaraSpriteRendererProperties}
   {rendererClass=&NiagaraSpriteRendererProperties, rendererIndex=1}
  ]
 }
 SmokeLWE {rendererClasses [], renderers []}
}
summary {
 systemName=NS_Emote_InkHoop_Stomps
 emitters [{emitterName=Impact_Spikes, rendererClasses [&NiagaraSpriteRendererProperties]}, {emitterName=SmokeLWE, rendererClasses []}]
 userVariables []
}
